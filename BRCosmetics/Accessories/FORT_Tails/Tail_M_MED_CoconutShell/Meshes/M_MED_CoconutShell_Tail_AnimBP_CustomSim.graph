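# Animation Blueprint for the CoconutShell cosmetic tail, posing and simulating its bone chain.
# Custom Idle selects between the input pose and a FullBody state machine (idle, locomotion, crouch, skydive/parachute, DBNO) that uses Neon Cat tail animations in visible states.
# Modify Bone nodes then reshape the tail, feeding per-joint AnimDynamics with world-space springs, gravity and angular limits.
# Several other Modify Bone nodes are left disconnected from the output.
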AnimBlueprint /BRCosmetics/Accessories/FORT_Tails/Tail_M_MED_CoconutShell/Meshes/M_MED_CoconutShell_Tail_AnimBP_CustomSim
target_skeleton=/Game/Accessories/FORT_Tails/Common/Fortnite_Base_Tail_Skeleton
graphs [
 AnimGraph {nodes [
   n1 = "Transform (Modify) Bone - Bone: C_Tail_A_Base_Jnt"(Alpha=1, Translation="0,0,0", Rotation="0,-40,0", Scale="1,1,1") {
    rotationMode=BMM_Replace
    rotationSpace=BCS_BoneSpace
    translationMode=BMM_Additive
    alphaBoolBlend {blendOption=Linear}
    alphaScaleBiasClamp {bInterpResult=true, interpSpeedIncreasing=0.5}
    boneToModify {boneName=C_Tail_A_Base_Jnt}
    rotation {yaw=-40}
   }
   n2 = "Transform (Modify) Bone - Bone: C_Tail_A_1_Jnt"(ComponentPose=@n1, Alpha=1, Translation="0,0,0", Rotation="-10,0,0", Scale="1,1,1") {
    rotationMode=BMM_Replace
    rotationSpace=BCS_BoneSpace
    translationMode=BMM_Additive
    alphaBoolBlend {blendOption=Linear}
    alphaScaleBiasClamp {bInterpResult=true, interpSpeedIncreasing=0.5}
    boneToModify {boneName=C_Tail_A_1_Jnt}
    rotation {pitch=-10}
   }
   n3 = "Transform (Modify) Bone - Bone: C_Tail_A_2_Jnt"(ComponentPose=@n2, Alpha=1, Translation="0,0,0", Rotation="10,0,0", Scale="1,1,1") {
    rotationMode=BMM_Replace
    rotationSpace=BCS_BoneSpace
    translationMode=BMM_Additive
    alphaBoolBlend {blendOption=Linear}
    alphaScaleBiasClamp {bInterpResult=true, interpSpeedIncreasing=0.5}
    boneToModify {boneName=C_Tail_A_2_Jnt}
    rotation {pitch=10}
   }
   n4 = "Transform (Modify) Bone - Bone: C_Tail_A_3_Jnt"(ComponentPose=@n3, Alpha=1, Translation="0,0,0", Rotation="18,0,0", Scale="1,1,1") {
    rotationMode=BMM_Replace
    rotationSpace=BCS_BoneSpace
    translationMode=BMM_Additive
    alphaBoolBlend {blendOption=Linear}
    alphaScaleBiasClamp {bInterpResult=true, interpSpeedIncreasing=0.5}
    boneToModify {boneName=C_Tail_A_3_Jnt}
    rotation {pitch=18}
   }
   n5 = "Transform (Modify) Bone - Bone: C_Tail_A_4_Jnt"(ComponentPose=@n4, Alpha=1, Translation="0,0,0", Rotation="23,0,0", Scale="1,1,1") {
    rotationMode=BMM_Replace
    rotationSpace=BCS_BoneSpace
    translationMode=BMM_Additive
    alphaBoolBlend {blendOption=Linear}
    alphaScaleBiasClamp {bInterpResult=true, interpSpeedIncreasing=0.5}
    boneToModify {boneName=C_Tail_A_4_Jnt}
    rotation {pitch=23}
   }
   n6 = "Transform (Modify) Bone - Bone: C_Tail_A_5_Jnt"(ComponentPose=@n5, Alpha=1, Translation="0,0,0", Rotation="0.7,0,0", Scale="1,1,1") {
    rotationMode=BMM_Replace
    rotationSpace=BCS_BoneSpace
    translationMode=BMM_Additive
    alphaBoolBlend {blendOption=Linear}
    alphaScaleBiasClamp {bInterpResult=true, interpSpeedIncreasing=0.5}
    boneToModify {boneName=C_Tail_A_5_Jnt}
    rotation {pitch=0.7}
   }
   n7 = "Transform (Modify) Bone - Bone: C_Tail_A_6_Jnt"(ComponentPose=@n6, Alpha=1, Translation="0,0,0", Rotation="-6,0,0", Scale="1,1,1") {
    rotationMode=BMM_Replace
    rotationSpace=BCS_BoneSpace
    translationMode=BMM_Additive
    alphaBoolBlend {blendOption=Linear}
    alphaScaleBiasClamp {bInterpResult=true, interpSpeedIncreasing=0.5}
    boneToModify {boneName=C_Tail_A_6_Jnt}
    rotation {pitch=-6}
   }
   n8 = "Transform (Modify) Bone - Bone: C_Tail_A_7_Jnt"(ComponentPose=@n7, Alpha=1, Translation="0,0,0", Rotation="0,0,0", Scale="1,1,1") {
    rotationMode=BMM_Replace
    rotationSpace=BCS_BoneSpace
    translationMode=BMM_Additive
    alphaBoolBlend {blendOption=Linear}
    alphaScaleBiasClamp {bInterpResult=true, interpSpeedIncreasing=0.5}
    boneToModify {boneName=C_Tail_A_7_Jnt}
   }
   n9 = "Transform (Modify) Bone - Bone: C_Tail_A_8_Jnt"(ComponentPose=@n8, Alpha=1, Translation="0,0,0", Rotation="14,0,0", Scale="1,1,1") {
    rotationMode=BMM_Replace
    rotationSpace=BCS_BoneSpace
    translationMode=BMM_Additive
    alphaBoolBlend {blendOption=Linear}
    alphaScaleBiasClamp {bInterpResult=true, interpSpeedIncreasing=0.5}
    boneToModify {boneName=C_Tail_A_8_Jnt}
    rotation {pitch=14}
   }
   "Transform (Modify) Bone - Bone: C_Tail_A_9_Jnt"(ComponentPose=@n9, Alpha=1, Translation="0,0,0", Rotation="-10,0,0", Scale="1,1,1") {
    rotationMode=BMM_Replace
    rotationSpace=BCS_BoneSpace
    translationMode=BMM_Additive
    alphaBoolBlend {blendOption=Linear}
    alphaScaleBiasClamp {bInterpResult=true, interpSpeedIncreasing=0.5}
    boneToModify {boneName=C_Tail_A_9_Jnt}
    rotation {pitch=-10}
   }
   n10 = "Transform (Modify) Bone - Bone: C_Tail_A_Base_Jnt"(Alpha=1, Translation="0,0,0", Rotation="0,25,0", Scale="1,1,1") {
    rotationMode=BMM_Replace
    rotationSpace=BCS_BoneSpace
    translationMode=BMM_Additive
    translationSpace=BCS_BoneSpace
    alphaBoolBlend {blendOption=Linear}
    alphaScaleBiasClamp {bInterpResult=true}
    boneToModify {boneName=C_Tail_A_Base_Jnt}
    rotation {yaw=25}
   }
   "Transform (Modify) Bone - Bone: C_Tail_A_Base_Jnt"(ComponentPose=@n10, Alpha=1, Translation="0,0,0", Rotation="0,-25,0", Scale="1,1,1") {
    rotationMode=BMM_Replace
    rotationSpace=BCS_BoneSpace
    translationMode=BMM_Additive
    translationSpace=BCS_BoneSpace
    alphaBoolBlend {blendOption=Linear}
    alphaScaleBiasClamp {bInterpResult=true}
    boneToModify {boneName=C_Tail_A_Base_Jnt}
    rotation {yaw=-25}
   }
   "Transform (Modify) Bone - Bone: C_Tail_A_Base_Jnt"(Alpha=1, Translation="0,0,0", Rotation="0,0,0", Scale="1,1,1") {
    rotationMode=BMM_Additive
    rotationSpace=BCS_BoneSpace
    translationMode=BMM_Additive
    alphaBoolBlend {blendOption=Linear}
    alphaScaleBiasClamp {bInterpResult=true, interpSpeedDecreasing=40, interpSpeedIncreasing=40}
    boneToModify {boneName=C_Tail_A_Base_Jnt}
   }
   "Transform (Modify) Bone - Bone: C_Tail_A_Base_Jnt"!(ComponentPose, bAlphaBoolEnabled:bool=True, Translation="0,0,0", Rotation="0,0,0", Scale="1,1,1") -> (Pose) {
    alphaInputType=Bool
    rotationMode=BMM_Additive
    rotationSpace=BCS_BoneSpace
    translationMode=BMM_Additive
    alphaBoolBlend {blendInTime=1, blendOption=Linear, blendOutTime=1}
    alphaScaleBiasClamp {bInterpResult=true, interpSpeedIncreasing=0.5}
    boneToModify {boneName=C_Tail_A_Base_Jnt}
   }
   n11 = Input Pose()
   AnimGraphNode_StateMachine_1 = FullBody() {is=title AnimGraphNode_StateMachine}
   n12 = Get Custom Idle() -> (Custom Idle:bool)
   n13 = "Transform (Modify) Bone - Bone: C_Tail_A_1_Jnt"(
    ComponentPose=Local To Component(LocalPose=Blend Poses by bool(BlendPose_0=@n11, BlendPose_1=@AnimGraphNode_StateMachine_1, bActiveValue=@n12) {blendType=Linear})
    Alpha=1
    Translation="0,0,0"
    Rotation="4,0,0"
    Scale="1,1,1"
   ) {
    rotationMode=BMM_Additive
    rotationSpace=BCS_BoneSpace
    translationMode=BMM_Additive
    alphaBoolBlend {blendOption=Linear}
    alphaScaleBiasClamp {bInterpResult=true, interpSpeedDecreasing=40, interpSpeedIncreasing=40}
    boneToModify {boneName=C_Tail_A_1_Jnt}
    rotation {pitch=4}
   }
   n14 = "Transform (Modify) Bone - Bone: C_Tail_A_Base_Jnt"!(ComponentPose=@n13, Alpha:float=1, Translation="0,0,0", Rotation="0,-12,0", Scale="1,1,1") -> (Pose) {
    rotationMode=BMM_Additive
    rotationSpace=BCS_BoneSpace
    translationMode=BMM_Additive
    alphaBoolBlend {blendOption=Linear}
    alphaScaleBiasClamp {bInterpResult=true, interpSpeedDecreasing=40, interpSpeedIncreasing=40}
    boneToModify {boneName=C_Tail_A_Base_Jnt}
    rotation {yaw=-12}
   }
   n15 = "Transform (Modify) Bone - Bone: C_Tail_A_Base_Jnt"(ComponentPose=@n14, Alpha=1, Translation="0,0,0", Rotation="0,20,0", Scale="1,1,1") {
    rotationMode=BMM_Replace
    rotationSpace=BCS_BoneSpace
    translationMode=BMM_Additive
    alphaBoolBlend {blendOption=Linear}
    alphaScaleBiasClamp {bInterpResult=true, interpSpeedIncreasing=4}
    boneToModify {boneName=C_Tail_A_Base_Jnt}
    rotation {yaw=20}
   }
   n16 = "Transform (Modify) Bone - Bone: C_Tail_A_1_Jnt"(ComponentPose=@n15, Alpha=1, Translation="0,0,0", Rotation="20,0,0", Scale="1,1,1") {
    rotationMode=BMM_Replace
    rotationSpace=BCS_BoneSpace
    translationMode=BMM_Additive
    alphaBoolBlend {blendOption=Linear}
    alphaScaleBiasClamp {bInterpResult=true, interpSpeedIncreasing=4}
    boneToModify {boneName=C_Tail_A_1_Jnt}
    rotation {pitch=20}
   }
   n17 = "Transform (Modify) Bone - Bone: C_Tail_A_2_Jnt"(ComponentPose=@n16, Alpha=1, Translation="0,0,0", Rotation="20,0,0", Scale="1,1,1") {
    rotationMode=BMM_Replace
    rotationSpace=BCS_BoneSpace
    translationMode=BMM_Additive
    alphaBoolBlend {blendOption=Linear}
    alphaScaleBiasClamp {bInterpResult=true, interpSpeedIncreasing=4}
    boneToModify {boneName=C_Tail_A_2_Jnt}
    rotation {pitch=20}
   }
   n18 = "Transform (Modify) Bone - Bone: C_Tail_A_3_Jnt"(ComponentPose=@n17, Alpha=1, Translation="0,0,0", Rotation="10,0,0", Scale="1,1,1") {
    rotationMode=BMM_Replace
    rotationSpace=BCS_BoneSpace
    translationMode=BMM_Additive
    alphaBoolBlend {blendOption=Linear}
    alphaScaleBiasClamp {bInterpResult=true, interpSpeedIncreasing=4}
    boneToModify {boneName=C_Tail_A_3_Jnt}
    rotation {pitch=10}
   }
   n19 = "Transform (Modify) Bone - Bone: C_Tail_A_4_Jnt"(ComponentPose=@n18, Alpha=1, Translation="0,0,0", Rotation="10,0,0", Scale="1,1,1") {
    rotationMode=BMM_Replace
    rotationSpace=BCS_BoneSpace
    translationMode=BMM_Additive
    alphaBoolBlend {blendOption=Linear}
    alphaScaleBiasClamp {bInterpResult=true, interpSpeedIncreasing=4}
    boneToModify {boneName=C_Tail_A_4_Jnt}
    rotation {pitch=10}
   }
   n20 = "Transform (Modify) Bone - Bone: C_Tail_A_5_Jnt"(ComponentPose=@n19, Alpha=1, Translation="0,0,0", Rotation="-10,0,0", Scale="1,1,1") {
    rotationMode=BMM_Replace
    rotationSpace=BCS_BoneSpace
    translationMode=BMM_Additive
    alphaBoolBlend {blendOption=Linear}
    alphaScaleBiasClamp {bInterpResult=true, interpSpeedIncreasing=4}
    boneToModify {boneName=C_Tail_A_5_Jnt}
    rotation {pitch=-10}
   }
   n21 = "Transform (Modify) Bone - Bone: C_Tail_A_6_Jnt"(ComponentPose=@n20, Alpha=1, Translation="0,0,0", Rotation="0,0,0", Scale="1,1,1") {
    rotationMode=BMM_Replace
    rotationSpace=BCS_BoneSpace
    translationMode=BMM_Additive
    alphaBoolBlend {blendOption=Linear}
    alphaScaleBiasClamp {bInterpResult=true, interpSpeedIncreasing=4}
    boneToModify {boneName=C_Tail_A_6_Jnt}
   }
   n22 = "Transform (Modify) Bone - Bone: C_Tail_A_7_Jnt"(ComponentPose=@n21, Alpha=1, Translation="0,0,0", Rotation="0,0,0", Scale="1,1,1") {
    rotationMode=BMM_Replace
    rotationSpace=BCS_BoneSpace
    translationMode=BMM_Additive
    alphaBoolBlend {blendOption=Linear}
    alphaScaleBiasClamp {bInterpResult=true, interpSpeedIncreasing=4}
    boneToModify {boneName=C_Tail_A_7_Jnt}
   }
   n23 = "Transform (Modify) Bone - Bone: C_Tail_A_8_Jnt"(ComponentPose=@n22, Alpha=1, Translation="0,0,0", Rotation="0,0,0", Scale="1,1,1") {
    rotationMode=BMM_Replace
    rotationSpace=BCS_BoneSpace
    translationMode=BMM_Additive
    alphaBoolBlend {blendOption=Linear}
    alphaScaleBiasClamp {bInterpResult=true, interpSpeedIncreasing=4}
    boneToModify {boneName=C_Tail_A_8_Jnt}
   }
   n24 = "Transform (Modify) Bone - Bone: C_Tail_A_9_Jnt"(ComponentPose=@n23, Alpha=1, Translation="0,0,0", Rotation="0,0,0", Scale="1,1,1") {
    rotationMode=BMM_Replace
    rotationSpace=BCS_BoneSpace
    translationMode=BMM_Additive
    alphaBoolBlend {blendOption=Linear}
    alphaScaleBiasClamp {bInterpResult=true, interpSpeedIncreasing=4}
    boneToModify {boneName=C_Tail_A_9_Jnt}
   }
   n25 = "Transform (Modify) Bone - Bone: C_Tail_A_Base_Jnt"(ComponentPose=@n24, Alpha=1, Translation="0,0,0", Rotation="0,20,0", Scale="1,1,1") {
    rotationMode=BMM_Replace
    rotationSpace=BCS_BoneSpace
    translationMode=BMM_Additive
    alphaBoolBlend {blendOption=Linear}
    alphaScaleBiasClamp {bInterpResult=true, interpSpeedDecreasing=30, interpSpeedIncreasing=4}
    boneToModify {boneName=C_Tail_A_Base_Jnt}
    rotation {yaw=20}
   }
   n26 = "Transform (Modify) Bone - Bone: C_Tail_A_Base_Jnt"(ComponentPose=@n25, Alpha=1, Translation="0,0,0", Rotation="0,6,0", Scale="1,1,1") {
    rotationMode=BMM_Replace
    rotationSpace=BCS_BoneSpace
    translationMode=BMM_Additive
    alphaBoolBlend {blendOption=Linear}
    alphaScaleBiasClamp {bInterpResult=true, interpSpeedDecreasing=30, interpSpeedIncreasing=4}
    boneToModify {boneName=C_Tail_A_Base_Jnt}
    rotation {yaw=6}
   }
   n27 = "Transform (Modify) Bone - Bone: C_Tail_A_Base_Jnt"(ComponentPose=@n26, Alpha=1, Translation="0,0,0", Rotation="0,20,0", Scale="1,1,1") {
    rotationMode=BMM_Replace
    rotationSpace=BCS_BoneSpace
    translationMode=BMM_Additive
    alphaBoolBlend {blendOption=Linear}
    alphaScaleBiasClamp {bInterpResult=true, interpSpeedDecreasing=30, interpSpeedIncreasing=4}
    boneToModify {boneName=C_Tail_A_Base_Jnt}
    rotation {yaw=20}
   }
   n28 = "Transform (Modify) Bone - Bone: C_Tail_A_1_Jnt"!(ComponentPose=@n27, bAlphaBoolEnabled:bool=True, Translation="0,0,0", Rotation="6,0,0", Scale="1,1,1") -> (Pose) {
    alphaInputType=Bool
    rotationMode=BMM_Additive
    rotationSpace=BCS_BoneSpace
    translationMode=BMM_Additive
    alphaBoolBlend {blendInTime=1, blendOption=Linear, blendOutTime=1}
    alphaScaleBiasClamp {bInterpResult=true, interpSpeedIncreasing=0.5}
    boneToModify {boneName=C_Tail_A_1_Jnt}
    rotation {pitch=6}
   }
   n29 = "Transform (Modify) Bone - Bone: C_Tail_A_2_Jnt"!(ComponentPose=@n28, bAlphaBoolEnabled:bool=True, Translation="0,0,0", Rotation="10,0,0", Scale="1,1,1") -> (Pose) {
    alphaInputType=Bool
    rotationMode=BMM_Additive
    rotationSpace=BCS_BoneSpace
    translationMode=BMM_Additive
    alphaBoolBlend {blendInTime=1, blendOption=Linear, blendOutTime=1}
    alphaScaleBiasClamp {bInterpResult=true, interpSpeedIncreasing=0.5}
    boneToModify {boneName=C_Tail_A_2_Jnt}
    rotation {pitch=10}
   }
   n30 = "Transform (Modify) Bone - Bone: C_Tail_A_3_Jnt"!(ComponentPose=@n29, bAlphaBoolEnabled:bool=True, Translation="0,0,0", Rotation="20,0,0", Scale="1,1,1") -> (Pose) {
    alphaInputType=Bool
    rotationMode=BMM_Additive
    rotationSpace=BCS_BoneSpace
    translationMode=BMM_Additive
    alphaBoolBlend {blendInTime=1, blendOption=Linear, blendOutTime=1}
    alphaScaleBiasClamp {bInterpResult=true, interpSpeedIncreasing=0.5}
    boneToModify {boneName=C_Tail_A_3_Jnt}
    rotation {pitch=20}
   }
   n31 = "Transform (Modify) Bone - Bone: C_Tail_A_4_Jnt"!(ComponentPose=@n30, bAlphaBoolEnabled:bool=True, Translation="0,0,0", Rotation="20,0,0", Scale="1,1,1") -> (Pose) {
    alphaInputType=Bool
    rotationMode=BMM_Additive
    rotationSpace=BCS_BoneSpace
    translationMode=BMM_Additive
    alphaBoolBlend {blendInTime=1, blendOption=Linear, blendOutTime=1}
    alphaScaleBiasClamp {bInterpResult=true, interpSpeedIncreasing=0.5}
    boneToModify {boneName=C_Tail_A_4_Jnt}
    rotation {pitch=20}
   }
   n32 = "Transform (Modify) Bone - Bone: C_Tail_A_5_Jnt"!(ComponentPose=@n31, bAlphaBoolEnabled:bool=True, Translation="0,0,0", Rotation="20,0,0", Scale="1,1,1") -> (Pose) {
    alphaInputType=Bool
    rotationMode=BMM_Additive
    rotationSpace=BCS_BoneSpace
    translationMode=BMM_Additive
    alphaBoolBlend {blendInTime=1, blendOption=Linear, blendOutTime=1}
    alphaScaleBiasClamp {bInterpResult=true, interpSpeedIncreasing=0.5}
    boneToModify {boneName=C_Tail_A_5_Jnt}
    rotation {pitch=20}
   }
   n33 = "Transform (Modify) Bone - Bone: C_Tail_A_6_Jnt"!(ComponentPose=@n32, bAlphaBoolEnabled:bool=True, Translation="0,0,0", Rotation="20,0,0", Scale="1,1,1") -> (Pose) {
    alphaInputType=Bool
    rotationMode=BMM_Additive
    rotationSpace=BCS_BoneSpace
    translationMode=BMM_Additive
    alphaBoolBlend {blendInTime=1, blendOption=Linear, blendOutTime=1}
    alphaScaleBiasClamp {bInterpResult=true, interpSpeedIncreasing=0.5}
    boneToModify {boneName=C_Tail_A_6_Jnt}
    rotation {pitch=20}
   }
   n34 = "Transform (Modify) Bone - Bone: C_Tail_A_7_Jnt"!(ComponentPose=@n33, bAlphaBoolEnabled:bool=True, Translation="0,0,0", Rotation="20,0,0", Scale="1,1,1") -> (Pose) {
    alphaInputType=Bool
    rotationMode=BMM_Additive
    rotationSpace=BCS_BoneSpace
    translationMode=BMM_Additive
    alphaBoolBlend {blendInTime=1, blendOption=Linear, blendOutTime=1}
    alphaScaleBiasClamp {bInterpResult=true, interpSpeedIncreasing=0.5}
    boneToModify {boneName=C_Tail_A_7_Jnt}
    rotation {pitch=20}
   }
   n35 = "Transform (Modify) Bone - Bone: C_Tail_A_8_Jnt"!(ComponentPose=@n34, bAlphaBoolEnabled:bool=True, Translation="0,0,0", Rotation="20,0,0", Scale="1,1,1") -> (Pose) {
    alphaInputType=Bool
    rotationMode=BMM_Additive
    rotationSpace=BCS_BoneSpace
    translationMode=BMM_Additive
    alphaBoolBlend {blendInTime=1, blendOption=Linear, blendOutTime=1}
    alphaScaleBiasClamp {bInterpResult=true, interpSpeedIncreasing=0.5}
    boneToModify {boneName=C_Tail_A_8_Jnt}
    rotation {pitch=20}
   }
   n36 = "Transform (Modify) Bone - Bone: C_Tail_A_Base_Jnt"(ComponentPose=@n35, Alpha=1, Translation="0,0,0", Rotation="0,10,0", Scale="1,1,1") {
    rotationMode=BMM_Additive
    rotationSpace=BCS_BoneSpace
    alphaBoolBlend {blendOption=Linear}
    boneToModify {boneName=C_Tail_A_Base_Jnt}
    rotation {yaw=10}
   }
   n37 = Anim Dynamics - Bone: C_Tail_A_1_Jnt!(ComponentPose=@n36, Alpha=1) -> (Pose) {
    angularDampingOverride=0.999
    angularSpringConstant=700
    bAngularSpring=true
    bOverrideAngularDamping=true
    bUseGravityOverride=true
    gravityScale=0.35
    simulationSpace=World
    alphaBoolBlend {blendOption=Linear}
    alphaScaleBiasClamp {bInterpResult=true}
    boundBone {boneName=C_Tail_A_1_Jnt}
    componentAppliedLinearAccClamp=<1000 1000 1000>
    componentLinearAccScale=<1 1 1>
    componentLinearVelScale=<1 1 1>
    gravityOverride {z=-980}
    physicsBodyDefinitions [{
      boundBone {boneName=C_Tail_A_1_Jnt}
      boxExtents=<12 12 12>
      constraintSetup {angularLimitsMax=<0 3 25>, angularLimitsMin=<0 -50 -25>, angularTarget=<1 0 0>}
      localJointOffset=<0.5 0 0>
     }]
   }
   n38 = Anim Dynamics - Bone: C_Tail_A_2_Jnt!(ComponentPose=@n37, Alpha=1) -> (Pose) {
    angularDampingOverride=0.999
    angularSpringConstant=700
    bAngularSpring=true
    bOverrideAngularDamping=true
    bUseGravityOverride=true
    gravityScale=0.35
    simulationSpace=World
    alphaBoolBlend {blendOption=Linear}
    alphaScaleBiasClamp {bInterpResult=true}
    boundBone {boneName=C_Tail_A_2_Jnt}
    componentAppliedLinearAccClamp=<1000 1000 1000>
    componentLinearAccScale=<1 1 1>
    componentLinearVelScale=<1 1 1>
    gravityOverride {z=-980}
    physicsBodyDefinitions [{
      boundBone {boneName=C_Tail_A_2_Jnt}
      boxExtents=<12 12 12>
      constraintSetup {angularLimitsMax=<0 3 25>, angularLimitsMin=<0 -50 -25>, angularTarget=<1 0 0>}
      localJointOffset=<0.5 0 0>
     }]
   }
   n39 = Anim Dynamics - Bone: C_Tail_A_3_Jnt!(ComponentPose=@n38, Alpha=1) -> (Pose) {
    angularDampingOverride=0.999
    angularSpringConstant=700
    bAngularSpring=true
    bOverrideAngularDamping=true
    bUseGravityOverride=true
    gravityScale=0.35
    simulationSpace=World
    alphaBoolBlend {blendOption=Linear}
    alphaScaleBiasClamp {bInterpResult=true}
    boundBone {boneName=C_Tail_A_3_Jnt}
    componentAppliedLinearAccClamp=<1000 1000 1000>
    componentLinearAccScale=<1 1 1>
    componentLinearVelScale=<1 1 1>
    gravityOverride {z=-980}
    physicsBodyDefinitions [{
      boundBone {boneName=C_Tail_A_3_Jnt}
      boxExtents=<12 12 12>
      constraintSetup {angularLimitsMax=<0 3 25>, angularLimitsMin=<0 -50 -25>, angularTarget=<1 0 0>}
      localJointOffset=<0.5 0 0>
     }]
   }
   n40 = Anim Dynamics - Bone: C_Tail_A_4_Jnt!(ComponentPose=@n39, Alpha=1) -> (Pose) {
    angularDampingOverride=0.999
    angularSpringConstant=700
    bAngularSpring=true
    bOverrideAngularDamping=true
    bUseGravityOverride=true
    gravityScale=0.35
    simulationSpace=World
    alphaBoolBlend {blendOption=Linear}
    alphaScaleBiasClamp {bInterpResult=true}
    boundBone {boneName=C_Tail_A_4_Jnt}
    componentAppliedLinearAccClamp=<1000 1000 1000>
    componentLinearAccScale=<1 1 1>
    componentLinearVelScale=<1 1 1>
    gravityOverride {z=-980}
    physicsBodyDefinitions [{
      boundBone {boneName=C_Tail_A_4_Jnt}
      boxExtents=<12 12 12>
      constraintSetup {angularLimitsMax=<0 3 25>, angularLimitsMin=<0 -45 -25>, angularTarget=<1 0 0>}
      localJointOffset=<0.5 0 0>
     }]
   }
   n41 = Anim Dynamics - Bone: C_Tail_A_5_Jnt!(ComponentPose=@n40, Alpha=1) -> (Pose) {
    angularDampingOverride=0.999
    angularSpringConstant=700
    bAngularSpring=true
    bOverrideAngularDamping=true
    bUseGravityOverride=true
    gravityScale=0.35
    simulationSpace=World
    alphaBoolBlend {blendOption=Linear}
    alphaScaleBiasClamp {bInterpResult=true}
    boundBone {boneName=C_Tail_A_5_Jnt}
    componentAppliedLinearAccClamp=<1000 1000 1000>
    componentLinearAccScale=<1 1 1>
    componentLinearVelScale=<1 1 1>
    gravityOverride {z=-980}
    physicsBodyDefinitions [{
      boundBone {boneName=C_Tail_A_5_Jnt}
      boxExtents=<13 13 13>
      constraintSetup {angularLimitsMax=<0 3 25>, angularLimitsMin=<0 -40 -25>, angularTarget=<1 0 0>}
      localJointOffset=<0.5 0 0>
     }]
   }
   n42 = Anim Dynamics - Bone: C_Tail_A_6_Jnt!(ComponentPose=@n41, Alpha=1) -> (Pose) {
    angularDampingOverride=0.999
    angularSpringConstant=700
    bAngularSpring=true
    bOverrideAngularDamping=true
    bUseGravityOverride=true
    gravityScale=0.35
    simulationSpace=World
    alphaBoolBlend {blendOption=Linear}
    alphaScaleBiasClamp {bInterpResult=true}
    boundBone {boneName=C_Tail_A_6_Jnt}
    componentAppliedLinearAccClamp=<1000 1000 1000>
    componentLinearAccScale=<1 1 1>
    componentLinearVelScale=<1 1 1>
    gravityOverride {z=-980}
    physicsBodyDefinitions [{
      boundBone {boneName=C_Tail_A_6_Jnt}
      boxExtents=<14 14 14>
      constraintSetup {angularLimitsMax=<0 3 25>, angularLimitsMin=<0 -35 -25>, angularTarget=<1 0 0>}
      localJointOffset=<0.5 0 0>
     }]
   }
   n43 = Anim Dynamics - Bone: C_Tail_A_7_Jnt!(ComponentPose=@n42, Alpha=1) -> (Pose) {
    angularDampingOverride=0.999
    angularSpringConstant=700
    bAngularSpring=true
    bOverrideAngularDamping=true
    bUseGravityOverride=true
    gravityScale=0.35
    simulationSpace=World
    alphaBoolBlend {blendOption=Linear}
    alphaScaleBiasClamp {bInterpResult=true}
    boundBone {boneName=C_Tail_A_7_Jnt}
    componentAppliedLinearAccClamp=<1000 1000 1000>
    componentLinearAccScale=<1 1 1>
    componentLinearVelScale=<1 1 1>
    gravityOverride {z=-980}
    physicsBodyDefinitions [{
      boundBone {boneName=C_Tail_A_7_Jnt}
      boxExtents=<15 15 15>
      constraintSetup {angularLimitsMax=<0 3 25>, angularLimitsMin=<0 -30 -25>, angularTarget=<1 0 0>}
      localJointOffset=<0.5 0 0>
     }]
   }
   n44 = Anim Dynamics - Bone: C_Tail_A_8_Jnt!(ComponentPose=@n43, Alpha=1) -> (Pose) {
    angularDampingOverride=0.999
    angularSpringConstant=700
    bAngularSpring=true
    bOverrideAngularDamping=true
    bUseGravityOverride=true
    gravityScale=0.35
    simulationSpace=World
    alphaBoolBlend {blendOption=Linear}
    alphaScaleBiasClamp {bInterpResult=true}
    boundBone {boneName=C_Tail_A_8_Jnt}
    componentAppliedLinearAccClamp=<1000 1000 1000>
    componentLinearAccScale=<1 1 1>
    componentLinearVelScale=<1 1 1>
    gravityOverride {z=-980}
    physicsBodyDefinitions [{
      boundBone {boneName=C_Tail_A_8_Jnt}
      boxExtents=<16 16 16>
      constraintSetup {angularLimitsMax=<0 3 25>, angularLimitsMin=<0 -25 -25>, angularTarget=<1 0 0>}
      localJointOffset=<0.5 0 0>
     }]
   }
   n45 = Anim Dynamics - Bone: C_Tail_A_9_Jnt!(ComponentPose=@n44, Alpha=1) -> (Pose) {
    angularDampingOverride=0.999
    angularSpringConstant=700
    bAngularSpring=true
    bOverrideAngularDamping=true
    bUseGravityOverride=true
    gravityScale=0.35
    simulationSpace=World
    alphaBoolBlend {blendOption=Linear}
    alphaScaleBiasClamp {bInterpResult=true}
    boundBone {boneName=C_Tail_A_9_Jnt}
    componentAppliedLinearAccClamp=<1000 1000 1000>
    componentLinearAccScale=<1 1 1>
    componentLinearVelScale=<1 1 1>
    gravityOverride {z=-980}
    physicsBodyDefinitions [{
      boundBone {boneName=C_Tail_A_9_Jnt}
      boxExtents=<16 16 16>
      constraintSetup {angularLimitsMax=<0 3 25>, angularLimitsMin=<0 -25 -25>, angularTarget=<1 0 0>}
      localJointOffset=<0.5 0 0>
     }]
   }
   Output Pose(Result=Component To Local(ComponentPose=@n45))
  ]}
 FullBody {
  entry=idle
  path=AnimGraph.AnimGraphNode_StateMachine_1.FullBody
  ids=id
  nodes [
   AnimStateNode_18 = Crouching() {is=title AnimStateNode}
   AnimStateNode_20 = idle() {
    is=title AnimStateNode
    stateEntered {guid=#1}
    stateFullyBlended {guid=#2}
    stateLeft {guid=#3}
   }
   AnimStateNode_21 = Locomotion() {is=title AnimStateNode}
   AnimStateNode_34 = SkyDive() {is=title AnimStateNode}
   AnimStateNode_9 = DBNO() {
    is=title AnimStateNode
    stateEntered {guid=#4}
    stateFullyBlended {guid=#5}
    stateLeft {guid=#6}
   }
   AnimStateTransitionNode_0 = AnimStateTransitionNode()
   AnimStateTransitionNode_1 = AnimStateTransitionNode()
   AnimStateTransitionNode_32 = AnimStateTransitionNode()
   AnimStateTransitionNode_74 = AnimStateTransitionNode()
   AnimStateTransitionNode_79 = AnimStateTransitionNode()
   AnimStateTransitionNode_8 = AnimStateTransitionNode()
   AnimStateTransitionNode_80 = AnimStateTransitionNode()
   AnimStateTransitionNode_81 = AnimStateTransitionNode() {
    transitionEnd {guid=#7}
    transitionInterrupt {guid=#8}
    transitionStart {guid=#9}
   }
   AnimStateTransitionNode_90 = AnimStateTransitionNode() {
    transitionEnd {guid=#10}
    transitionInterrupt {guid=#11}
    transitionStart {guid=#12}
   }
   AnimStateTransitionNode_92 = AnimStateTransitionNode()
   AnimStateTransitionNode_93 = AnimStateTransitionNode()
   AnimStateTransitionNode_95 = AnimStateTransitionNode()
  ]
  links [
   {
    dst=Locomotion
    rule=AnimGraph.AnimGraphNode_StateMachine_1.FullBody.AnimStateTransitionNode_93.Transition
    src=Crouching
    via=AnimStateTransitionNode_93
   }
   {
    dst=idle
    rule=AnimGraph.AnimGraphNode_StateMachine_1.FullBody.AnimStateTransitionNode_95.Transition
    src=Crouching
    via=AnimStateTransitionNode_95
   }
   {
    dst=idle
    rule=AnimGraph.AnimGraphNode_StateMachine_1.FullBody.AnimStateTransitionNode_90.Transition
    src=DBNO
    via=AnimStateTransitionNode_90
   }
   {
    dst=Crouching
    rule=AnimGraph.AnimGraphNode_StateMachine_1.FullBody.AnimStateTransitionNode_92.Transition
    src=Locomotion
    via=AnimStateTransitionNode_92
   }
   {
    dst=SkyDive
    rule=AnimGraph.AnimGraphNode_StateMachine_1.FullBody.AnimStateTransitionNode_0.Transition
    src=Locomotion
    via=AnimStateTransitionNode_0
   }
   {
    dst=idle
    rule=AnimGraph.AnimGraphNode_StateMachine_1.FullBody.AnimStateTransitionNode_8.Transition
    src=Locomotion
    via=AnimStateTransitionNode_8
   }
   {
    dst=Locomotion
    rule=AnimGraph.AnimGraphNode_StateMachine_1.FullBody.AnimStateTransitionNode_1.Transition
    src=SkyDive
    via=AnimStateTransitionNode_1
   }
   {
    dst=idle
    rule=AnimGraph.AnimGraphNode_StateMachine_1.FullBody.AnimStateTransitionNode_79.Transition
    src=SkyDive
    via=AnimStateTransitionNode_79
   }
   {
    dst=Crouching
    rule=AnimGraph.AnimGraphNode_StateMachine_1.FullBody.AnimStateTransitionNode_32.Transition
    src=idle
    via=AnimStateTransitionNode_32
   }
   {
    dst=DBNO
    rule=AnimGraph.AnimGraphNode_StateMachine_1.FullBody.AnimStateTransitionNode_81.Transition
    src=idle
    via=AnimStateTransitionNode_81
   }
   {
    dst=Locomotion
    rule=AnimGraph.AnimGraphNode_StateMachine_1.FullBody.AnimStateTransitionNode_80.Transition
    src=idle
    via=AnimStateTransitionNode_80
   }
   {
    dst=SkyDive
    rule=AnimGraph.AnimGraphNode_StateMachine_1.FullBody.AnimStateTransitionNode_74.AnimationTransitionGraph_1
    src=idle
    via=AnimStateTransitionNode_74
   }
  ]
 }
 Crouching {
  path=AnimGraph.AnimGraphNode_StateMachine_1.FullBody.AnimStateNode_18.Crouching
  nodes [
   AnimGraphNode_StateMachine_0 = Crouch() {is=title AnimGraphNode_StateMachine}
   Output Animation Pose(Result=@AnimGraphNode_StateMachine_0)
  ]
 }
 Crouch {
  entry=Blank
  path=AnimGraph.AnimGraphNode_StateMachine_1.FullBody.AnimStateNode_18.Crouching.AnimGraphNode_StateMachine_0.Crouch
  ids=id
  nodes [
   AnimStateNode_0 = Blank() {is=title AnimStateNode}
   AnimStateNode_1 = Sliding() {is=title AnimStateNode}
   AnimStateNode_2 = CrouchSprinting() {is=title AnimStateNode}
   AnimStateNode_27 = CrouchWalk() {is=title AnimStateNode}
   AnimStateNode_29 = Crouch() {is=title AnimStateNode}
   AnimStateTransitionNode_0 = AnimStateTransitionNode()
   AnimStateTransitionNode_1 = AnimStateTransitionNode()
   AnimStateTransitionNode_11 = AnimStateTransitionNode()
   AnimStateTransitionNode_12 = AnimStateTransitionNode()
   AnimStateTransitionNode_2 = AnimStateTransitionNode()
   AnimStateTransitionNode_3 = AnimStateTransitionNode()
   AnimStateTransitionNode_4 = AnimStateTransitionNode()
   AnimStateTransitionNode_47 = AnimStateTransitionNode()
   AnimStateTransitionNode_55 = AnimStateTransitionNode()
   AnimStateTransitionNode_7 = AnimStateTransitionNode()
  ]
  links [
   {
    dst=Crouch
    rule=AnimGraph.AnimGraphNode_StateMachine_1.FullBody.AnimStateNode_18.Crouching.AnimGraphNode_StateMachine_0.Crouch.AnimStateTransitionNode_11.Transition
    src=Blank
    via=AnimStateTransitionNode_11
   }
   {
    dst=CrouchWalk
    rule=AnimGraph.AnimGraphNode_StateMachine_1.FullBody.AnimStateNode_18.Crouching.AnimGraphNode_StateMachine_0.Crouch.AnimStateTransitionNode_12.Transition
    src=Blank
    via=AnimStateTransitionNode_12
   }
   {
    dst=CrouchWalk
    rule=AnimGraph.AnimGraphNode_StateMachine_1.FullBody.AnimStateNode_18.Crouching.AnimGraphNode_StateMachine_0.Crouch.AnimStateTransitionNode_7.Transition
    src=Crouch
    via=AnimStateTransitionNode_7
   }
   {
    dst=Sliding
    rule=AnimGraph.AnimGraphNode_StateMachine_1.FullBody.AnimStateNode_18.Crouching.AnimGraphNode_StateMachine_0.Crouch.AnimStateTransitionNode_0.Transition
    src=Crouch
    via=AnimStateTransitionNode_0
   }
   {
    dst=CrouchWalk
    rule=AnimGraph.AnimGraphNode_StateMachine_1.FullBody.AnimStateNode_18.Crouching.AnimGraphNode_StateMachine_0.Crouch.AnimStateTransitionNode_55.AnimationTransitionGraph_1
    src=CrouchSprinting
    via=AnimStateTransitionNode_55
   }
   {
    dst=Crouch
    rule=AnimGraph.AnimGraphNode_StateMachine_1.FullBody.AnimStateNode_18.Crouching.AnimGraphNode_StateMachine_0.Crouch.AnimStateTransitionNode_4.Transition
    src=CrouchWalk
    via=AnimStateTransitionNode_4
   }
   {
    dst=CrouchSprinting
    rule=AnimGraph.AnimGraphNode_StateMachine_1.FullBody.AnimStateNode_18.Crouching.AnimGraphNode_StateMachine_0.Crouch.AnimStateTransitionNode_47.AnimationTransitionGraph_1
    src=CrouchWalk
    via=AnimStateTransitionNode_47
   }
   {
    dst=Sliding
    rule=AnimGraph.AnimGraphNode_StateMachine_1.FullBody.AnimStateNode_18.Crouching.AnimGraphNode_StateMachine_0.Crouch.AnimStateTransitionNode_3.Transition
    src=CrouchWalk
    via=AnimStateTransitionNode_3
   }
   {
    dst=Crouch
    rule=AnimGraph.AnimGraphNode_StateMachine_1.FullBody.AnimStateNode_18.Crouching.AnimGraphNode_StateMachine_0.Crouch.AnimStateTransitionNode_1.Transition
    src=Sliding
    via=AnimStateTransitionNode_1
   }
   {
    dst=CrouchWalk
    rule=AnimGraph.AnimGraphNode_StateMachine_1.FullBody.AnimStateNode_18.Crouching.AnimGraphNode_StateMachine_0.Crouch.AnimStateTransitionNode_2.Transition
    src=Sliding
    via=AnimStateTransitionNode_2
   }
  ]
 }
 Blank {
  path=AnimGraph.AnimGraphNode_StateMachine_1.FullBody.AnimStateNode_18.Crouching.AnimGraphNode_StateMachine_0.Crouch.AnimStateNode_0.Blank
  nodes [Output Animation Pose()]
 }
 Sliding {
  path=AnimGraph.AnimGraphNode_StateMachine_1.FullBody.AnimStateNode_18.Crouching.AnimGraphNode_StateMachine_0.Crouch.AnimStateNode_1.Sliding
  nodes [Output Animation Pose()]
 }
 CrouchSprinting {
  path=AnimGraph.AnimGraphNode_StateMachine_1.FullBody.AnimStateNode_18.Crouching.AnimGraphNode_StateMachine_0.Crouch.AnimStateNode_2.CrouchSprinting
  nodes [Output Animation Pose(Result=Sequence Player 'Tail_Crouch_Sprint'() {sequence=/Game/Accessories/FORT_Tails/Tail_Neon_Cat/Anims/Tail_Crouch_Sprint})]
 }
 CrouchWalk {
  path=AnimGraph.AnimGraphNode_StateMachine_1.FullBody.AnimStateNode_18.Crouching.AnimGraphNode_StateMachine_0.Crouch.AnimStateNode_27.CrouchWalk
  nodes [Output Animation Pose(Result=Sequence Player 'Tail_Crouch_Walk'() {sequence=/Game/Accessories/FORT_Tails/Tail_Neon_Cat/Anims/Tail_Crouch_Walk})]
 }
 Crouch {
  path=AnimGraph.AnimGraphNode_StateMachine_1.FullBody.AnimStateNode_18.Crouching.AnimGraphNode_StateMachine_0.Crouch.AnimStateNode_29.Crouch
  nodes [
   n1 = "Transform (Modify) Bone - Bone: C_Tail_A_3_Jnt"!(
    ComponentPose=Local To Component(LocalPose=Sequence Player 'Tail_Crouch_Idle'() {sequence=/Game/Accessories/FORT_Tails/Tail_Neon_Cat/Anims/Tail_Crouch_Idle})
    Alpha:float=1
    Translation="0,0,0"
    Rotation="20,0,0"
    Scale="1,1,1"
   ) -> (Pose) {
    rotationMode=BMM_Additive
    rotationSpace=BCS_BoneSpace
    alphaBoolBlend {blendOption=Linear}
    boneToModify {boneName=C_Tail_A_3_Jnt}
    rotation {pitch=20}
   }
   n2 = "Transform (Modify) Bone - Bone: C_Tail_A_5_Jnt"!(ComponentPose=@n1, Alpha:float=1, Translation="0,0,0", Rotation="20,0,0", Scale="1,1,1") -> (Pose) {
    rotationMode=BMM_Additive
    rotationSpace=BCS_BoneSpace
    alphaBoolBlend {blendOption=Linear}
    boneToModify {boneName=C_Tail_A_5_Jnt}
    rotation {pitch=20}
   }
   n3 = "Transform (Modify) Bone - Bone: C_Tail_A_7_Jnt"!(ComponentPose=@n2, Alpha:float=1, Translation="0,0,0", Rotation="20,0,0", Scale="1,1,1") -> (Pose) {
    rotationMode=BMM_Additive
    rotationSpace=BCS_BoneSpace
    alphaBoolBlend {blendOption=Linear}
    boneToModify {boneName=C_Tail_A_7_Jnt}
    rotation {pitch=20}
   }
   Output Animation Pose(Result=Component To Local(ComponentPose=@n3))
  ]
 }
 Transition {
  path=AnimGraph.AnimGraphNode_StateMachine_1.FullBody.AnimStateNode_18.Crouching.AnimGraphNode_StateMachine_0.Crouch.AnimStateTransitionNode_0.Transition
  nodes [Result(bCanEnterTransition=Get bIsSliding())]
 }
 Transition {
  path=AnimGraph.AnimGraphNode_StateMachine_1.FullBody.AnimStateNode_18.Crouching.AnimGraphNode_StateMachine_0.Crouch.AnimStateTransitionNode_1.Transition
  nodes [Result(bCanEnterTransition=NOT Boolean(A=Get bIsSliding()))]
 }
 Transition {
  path=AnimGraph.AnimGraphNode_StateMachine_1.FullBody.AnimStateNode_18.Crouching.AnimGraphNode_StateMachine_0.Crouch.AnimStateTransitionNode_11.Transition
  nodes [Result(bCanEnterTransition=NOT Boolean(A=Get bIsMoving2D()))]
 }
 Transition {
  path=AnimGraph.AnimGraphNode_StateMachine_1.FullBody.AnimStateNode_18.Crouching.AnimGraphNode_StateMachine_0.Crouch.AnimStateTransitionNode_12.Transition
  nodes [Result(bCanEnterTransition=Get bIsMoving2D())]
 }
 Transition {
  path=AnimGraph.AnimGraphNode_StateMachine_1.FullBody.AnimStateNode_18.Crouching.AnimGraphNode_StateMachine_0.Crouch.AnimStateTransitionNode_2.Transition
  nodes [Result(bCanEnterTransition=NOT Boolean(A=Get bIsSliding()))]
 }
 Transition {
  path=AnimGraph.AnimGraphNode_StateMachine_1.FullBody.AnimStateNode_18.Crouching.AnimGraphNode_StateMachine_0.Crouch.AnimStateTransitionNode_3.Transition
  nodes [Result(bCanEnterTransition=Get bIsSliding())]
 }
 Transition {
  path=AnimGraph.AnimGraphNode_StateMachine_1.FullBody.AnimStateNode_18.Crouching.AnimGraphNode_StateMachine_0.Crouch.AnimStateTransitionNode_4.Transition
  nodes [Result(bCanEnterTransition=NOT Boolean(A=Get bIsMoving2D()))]
 }
 AnimationTransitionGraph_1 {
  path=AnimGraph.AnimGraphNode_StateMachine_1.FullBody.AnimStateNode_18.Crouching.AnimGraphNode_StateMachine_0.Crouch.AnimStateTransitionNode_47.AnimationTransitionGraph_1
  nodes [Result(bCanEnterTransition=Get bIsCrouchSprinting())]
 }
 AnimationTransitionGraph_1 {
  path=AnimGraph.AnimGraphNode_StateMachine_1.FullBody.AnimStateNode_18.Crouching.AnimGraphNode_StateMachine_0.Crouch.AnimStateTransitionNode_55.AnimationTransitionGraph_1
  nodes [Result(bCanEnterTransition=NOT Boolean(A=Get bIsCrouchSprinting()))]
 }
 Transition {
  path=AnimGraph.AnimGraphNode_StateMachine_1.FullBody.AnimStateNode_18.Crouching.AnimGraphNode_StateMachine_0.Crouch.AnimStateTransitionNode_7.Transition
  nodes [Result(bCanEnterTransition=Get bIsMoving2D())]
 }
 idle {
  path=AnimGraph.AnimGraphNode_StateMachine_1.FullBody.AnimStateNode_20.idle
  nodes [Output Animation Pose(Result=Sequence Player 'Tail_Idle'() {sequence=/Game/Accessories/FORT_Tails/Tail_Neon_Cat/Anims/Tail_Idle})]
 }
 Locomotion {
  path=AnimGraph.AnimGraphNode_StateMachine_1.FullBody.AnimStateNode_21.Locomotion
  nodes [
   n1 = "Transform (Modify) Bone - Bone: C_Root_Main_Root_Jnt"(
    ComponentPose=Local To Component(LocalPose=Sequence Player 'Tail_Sprint'!(PlayRate:float=0.3) -> (Pose) {sequence=/Game/Accessories/FORT_Tails/Tail_Neon_Cat/Anims/Tail_Sprint})
    Alpha=1
    Translation="0,0,0"
    Rotation="0,0,10"
    Scale="1,1,1"
   ) {
    rotationMode=BMM_Additive
    rotationSpace=BCS_BoneSpace
    alphaBoolBlend {blendOption=Linear}
    boneToModify {boneName=C_Root_Main_Root_Jnt}
    rotation {roll=10}
   }
   Output Animation Pose(Result=Component To Local(ComponentPose=@n1))
   Get SpeedAdjustedPlayrate() -> (SpeedAdjustedPlayrate:float)
  ]
 }
 SkyDive {
  path=AnimGraph.AnimGraphNode_StateMachine_1.FullBody.AnimStateNode_34.SkyDive
  nodes [
   AnimGraphNode_StateMachine_0 = SkydiveParachute() {is=title AnimGraphNode_StateMachine}
   Output Animation Pose(Result=@AnimGraphNode_StateMachine_0)
  ]
 }
 SkydiveParachute {
  entry=SkyDive
  path=AnimGraph.AnimGraphNode_StateMachine_1.FullBody.AnimStateNode_34.SkyDive.AnimGraphNode_StateMachine_0.SkydiveParachute
  ids=id
  nodes [
   AnimStateNode_1 = Parachute() {is=title AnimStateNode}
   AnimStateNode_2 = SkyDive()
   AnimStateNode_28 = SkydiveStart() {is=title AnimStateNode}
   AnimStateTransitionNode_0 = AnimStateTransitionNode()
   AnimStateTransitionNode_1 = AnimStateTransitionNode()
   AnimStateTransitionNode_28 = AnimStateTransitionNode() {bAutomaticRuleBasedOnSequencePlayerInState=true}
  ]
  links [
   {
    dst=SkyDive
    rule=AnimGraph.AnimGraphNode_StateMachine_1.FullBody.AnimStateNode_34.SkyDive.AnimGraphNode_StateMachine_0.SkydiveParachute.AnimStateTransitionNode_1.Transition
    src=Parachute
    via=AnimStateTransitionNode_1
   }
   {
    dst=Parachute
    rule=AnimGraph.AnimGraphNode_StateMachine_1.FullBody.AnimStateNode_34.SkyDive.AnimGraphNode_StateMachine_0.SkydiveParachute.AnimStateTransitionNode_0.Transition
    src=SkyDive
    via=AnimStateTransitionNode_0
   }
   {
    dst=SkyDive
    rule=AnimGraph.AnimGraphNode_StateMachine_1.FullBody.AnimStateNode_34.SkyDive.AnimGraphNode_StateMachine_0.SkydiveParachute.AnimStateTransitionNode_28.Transition
    src=SkydiveStart
    via=AnimStateTransitionNode_28
   }
  ]
 }
 Parachute {
  path=AnimGraph.AnimGraphNode_StateMachine_1.FullBody.AnimStateNode_34.SkyDive.AnimGraphNode_StateMachine_0.SkydiveParachute.AnimStateNode_1.Parachute
  nodes [
   n1 = "Transform (Modify) Bone - Bone: C_Root_Main_Root_Jnt"(
    ComponentPose=Local To Component(LocalPose=Sequence Player 'Tail_Idle'() {sequence=&Tail_Idle})
    Alpha=1
    Translation="0,0,0"
    Rotation="0,0,20"
    Scale="1,1,1"
   ) {
    rotationMode=BMM_Additive
    rotationSpace=BCS_BoneSpace
    alphaBoolBlend {blendOption=Linear}
    boneToModify {boneName=C_Root_Main_Root_Jnt}
    rotation {roll=20}
   }
   Output Animation Pose(Result=Component To Local(ComponentPose=@n1))
  ]
 }
 SkyDive {
  path=AnimGraph.AnimGraphNode_StateMachine_1.FullBody.AnimStateNode_34.SkyDive.AnimGraphNode_StateMachine_0.SkydiveParachute.AnimStateNode_2.SkyDive
  nodes [
   AnimGraphNode_StateMachine_0 = Skydiving() {is=title AnimGraphNode_StateMachine}
   Output Animation Pose(Result=@AnimGraphNode_StateMachine_0)
  ]
 }
 Skydiving {
  entry=Passthrough
  path=AnimGraph.AnimGraphNode_StateMachine_1.FullBody.AnimStateNode_34.SkyDive.AnimGraphNode_StateMachine_0.SkydiveParachute.AnimStateNode_2.SkyDive.AnimGraphNode_StateMachine_0.Skydiving
  ids=id
  nodes [
   AnimStateNode_0 = Passthrough() {is=title AnimStateNode}
   AnimStateNode_1 = Glide() {is=title AnimStateNode}
   AnimStateNode_2 = Dive() {is=title AnimStateNode}
   AnimStateNode_3 = to_Glide() {is=title AnimStateNode}
   AnimStateNode_4 = to_Dive() {is=title AnimStateNode}
   AnimStateTransitionNode_0 = AnimStateTransitionNode()
   AnimStateTransitionNode_1 = AnimStateTransitionNode()
   AnimStateTransitionNode_2 = AnimStateTransitionNode() {bAutomaticRuleBasedOnSequencePlayerInState=true, blendMode=Linear, crossfadeDuration=0.085}
   AnimStateTransitionNode_3 = AnimStateTransitionNode()
   AnimStateTransitionNode_4 = AnimStateTransitionNode()
   AnimStateTransitionNode_6 = AnimStateTransitionNode() {priorityOrder=2}
   AnimStateTransitionNode_7 = AnimStateTransitionNode() {priorityOrder=2}
   AnimStateTransitionNode_8 = AnimStateTransitionNode() {bAutomaticRuleBasedOnSequencePlayerInState=true, crossfadeDuration=0.75}
  ]
  links [
   {
    dst=to_Glide
    rule=AnimGraph.AnimGraphNode_StateMachine_1.FullBody.AnimStateNode_34.SkyDive.AnimGraphNode_StateMachine_0.SkydiveParachute.AnimStateNode_2.SkyDive.AnimGraphNode_StateMachine_0.Skydiving.AnimStateTransitionNode_1.Transition
    src=Dive
    via=AnimStateTransitionNode_1
   }
   {
    dst=to_Dive
    rule=AnimGraph.AnimGraphNode_StateMachine_1.FullBody.AnimStateNode_34.SkyDive.AnimGraphNode_StateMachine_0.SkydiveParachute.AnimStateNode_2.SkyDive.AnimGraphNode_StateMachine_0.Skydiving.AnimStateTransitionNode_4.Transition
    src=Glide
    via=AnimStateTransitionNode_4
   }
   {
    dst=Dive
    rule=AnimGraph.AnimGraphNode_StateMachine_1.FullBody.AnimStateNode_34.SkyDive.AnimGraphNode_StateMachine_0.SkydiveParachute.AnimStateNode_2.SkyDive.AnimGraphNode_StateMachine_0.Skydiving.AnimStateTransitionNode_0.Transition
    src=Passthrough
    via=AnimStateTransitionNode_0
   }
   {
    dst=Glide
    rule=AnimGraph.AnimGraphNode_StateMachine_1.FullBody.AnimStateNode_34.SkyDive.AnimGraphNode_StateMachine_0.SkydiveParachute.AnimStateNode_2.SkyDive.AnimGraphNode_StateMachine_0.Skydiving.AnimStateTransitionNode_3.Transition
    src=Passthrough
    via=AnimStateTransitionNode_3
   }
   {
    dst=Dive
    rule=AnimGraph.AnimGraphNode_StateMachine_1.FullBody.AnimStateNode_34.SkyDive.AnimGraphNode_StateMachine_0.SkydiveParachute.AnimStateNode_2.SkyDive.AnimGraphNode_StateMachine_0.Skydiving.AnimStateTransitionNode_8.Transition
    src=to_Dive
    via=AnimStateTransitionNode_8
   }
   {
    dst=to_Glide
    rule=AnimGraph.AnimGraphNode_StateMachine_1.FullBody.AnimStateNode_34.SkyDive.AnimGraphNode_StateMachine_0.SkydiveParachute.AnimStateNode_2.SkyDive.AnimGraphNode_StateMachine_0.Skydiving.AnimStateTransitionNode_6.Transition
    src=to_Dive
    via=AnimStateTransitionNode_6
   }
   {
    dst=Glide
    rule=AnimGraph.AnimGraphNode_StateMachine_1.FullBody.AnimStateNode_34.SkyDive.AnimGraphNode_StateMachine_0.SkydiveParachute.AnimStateNode_2.SkyDive.AnimGraphNode_StateMachine_0.Skydiving.AnimStateTransitionNode_2.Transition
    src=to_Glide
    via=AnimStateTransitionNode_2
   }
   {
    dst=to_Dive
    rule=AnimGraph.AnimGraphNode_StateMachine_1.FullBody.AnimStateNode_34.SkyDive.AnimGraphNode_StateMachine_0.SkydiveParachute.AnimStateNode_2.SkyDive.AnimGraphNode_StateMachine_0.Skydiving.AnimStateTransitionNode_7.Transition
    src=to_Glide
    via=AnimStateTransitionNode_7
   }
  ]
 }
 Passthrough {
  path=AnimGraph.AnimGraphNode_StateMachine_1.FullBody.AnimStateNode_34.SkyDive.AnimGraphNode_StateMachine_0.SkydiveParachute.AnimStateNode_2.SkyDive.AnimGraphNode_StateMachine_0.Skydiving.AnimStateNode_0.Passthrough
  nodes [Output Animation Pose(Result=Local Space Ref Pose())]
 }
 Glide {
  path=AnimGraph.AnimGraphNode_StateMachine_1.FullBody.AnimStateNode_34.SkyDive.AnimGraphNode_StateMachine_0.SkydiveParachute.AnimStateNode_2.SkyDive.AnimGraphNode_StateMachine_0.Skydiving.AnimStateNode_1.Glide
  nodes [
   n1 = Apply Additive(
    Base=Sequence Player 'Tail_Glide'() {sequence=/Game/Accessories/FORT_Tails/Tail_Neon_Cat/Anims/Tail_Glide}
    Additive="Sequence Player (additive) 'Tail_Glide_Zero_Offset_Pose'"() {sequence=/Game/Accessories/FORT_Tails/Tail_Neon_Cat/Anims/Tail_Glide_Zero_Offset_Pose}
    Alpha=0.5
   ) {alphaBoolBlend {blendOption=Linear}}
   n2 = Break Fort Anim Input Skydiving!(FortAnimInput_Skydiving:"Fort Anim Input Skydiving Structure (by ref)"=Get SkydivingState() -> (SkydivingState:Fort Anim Input Skydiving Structure)) -> (
    ParaGilderRootOffset:Vector=~
    ParaGliderRootRotationOverride:Rotator=~
    bUseParaGlideRootModifier:bool=~
    bIsSkydivingFromLaunchPad:bool=~
    bIsSkydivingFromBus:bool=~
    bIsInVortex:bool=~
    bIsDiving:bool=~
    bIsParachuteOpen:bool=~
    bIsSkydiving:bool=~
    bIsParachuteLeaning:bool=~
    bIsSkydiveLeaning:bool=~
    bIsLeaning:bool=~
    bIsSkydiveDiveMode:bool=~
    bPlaySkydiveDrift:bool=~
    bSkydiveDriftDelayActive:bool=~
    bSkydiveDriftAnimAllowed:bool=~
    bIsGliderRight:bool=~
    bIsGliderCenter:bool=~
    bIsGliderLeft:bool=~
    bIsGliderForward:bool=~
    bIsGliderBack:bool=~
    bHasDeployedAndIsStandingUpright:bool=~
    LocalAccelForward:float=~
    LocalAccelRight:float=~
    LocalVelocityRight:float=~
    SkydiveAimPitch:float=~
    SkydiveAimYaw:float=~
    DeployChuteAnimRate:float=~
    SkydiveDriftAnimRate:float=~
    SkydiveDriftAnimRateCurrent:float=~
    SkydiveFidgetAnimRate:float=~
    SkydiveFidgetAnimRateCurrent:float=~
    SkydiveAdditiveAlpha:float=~
    SkydiveDriftDelay:float=~
    SkydiveDriftAnim:int
    SkydiveDriftAnimMax:int
    LaunchpadAnim:int
    LocalAccelDir:ESkydivingDirection Enum
    DirectionLast:ESkydivingDirection Enum
    PoseAdditive:Anim Sequence Object Reference
   ) {is=title K2Node_BreakStruct}
   n3 = Blendspace Player 'BS_Tail_Glide_Turn_Additive'(X=Get TailSkydiveYaw() -> (TailSkydiveYaw:float), Y=@n2.SkydiveAimPitch) {blendSpace=/Game/Accessories/FORT_Tails/Tail_Neon_Cat/Anims/BS_Tail_Glide_Turn_Additive}
   n4 = Get TailSkydiveLean() -> (TailSkydiveLean:float)
   Output Animation Pose(Result=Apply Additive!(
     Base=Apply Additive!(Base=@n1, Additive=@n3, Alpha:float) -> (Pose) {alphaBoolBlend {blendOption=Linear}}
     Additive=Blendspace Player 'BS_Tail_Glide_Lean_Additive'(X=@n4, Y=@n2.SkydiveAimPitch) {blendSpace=/Game/Accessories/FORT_Tails/Tail_Neon_Cat/Anims/BS_Tail_Glide_Lean_Additive}
     bAlphaBoolEnabled:bool=@n2.bIsSkydiveLeaning
    ) -> (Pose) {alphaInputType=Bool, alphaBoolBlend {blendInTime=0.5, blendOption=Linear, blendOutTime=0.5}})
  ]
 }
 Dive {
  path=AnimGraph.AnimGraphNode_StateMachine_1.FullBody.AnimStateNode_34.SkyDive.AnimGraphNode_StateMachine_0.SkydiveParachute.AnimStateNode_2.SkyDive.AnimGraphNode_StateMachine_0.Skydiving.AnimStateNode_2.Dive
  nodes [
   n1 = Break Fort Anim Input Skydiving(FortAnimInput_Skydiving=Get SkydivingState()) -> (
    ParaGilderRootOffset=~
    ParaGliderRootRotationOverride=~
    bUseParaGlideRootModifier=~
    bIsSkydivingFromLaunchPad=~
    bIsSkydivingFromBus=~
    bIsInVortex=~
    bIsDiving=~
    bIsParachuteOpen=~
    bIsSkydiving=~
    bIsParachuteLeaning=~
    bIsSkydiveLeaning=~
    bIsLeaning=~
    bIsSkydiveDiveMode=~
    bPlaySkydiveDrift=~
    bSkydiveDriftDelayActive=~
    bSkydiveDriftAnimAllowed=~
    bIsGliderRight=~
    bIsGliderCenter=~
    bIsGliderLeft=~
    bIsGliderForward=~
    bIsGliderBack=~
    bHasDeployedAndIsStandingUpright=~
    LocalAccelForward=~
    LocalAccelRight=~
    LocalVelocityRight=~
    SkydiveAimPitch=~
    SkydiveAimYaw=~
    DeployChuteAnimRate=~
    SkydiveDriftAnimRate=~
    SkydiveDriftAnimRateCurrent=~
    SkydiveFidgetAnimRate=~
    SkydiveFidgetAnimRateCurrent=~
    SkydiveAdditiveAlpha=~
    SkydiveDriftDelay=~
   )
   n2 = Get TailDiveSpeed() -> (TailDiveSpeed:float)
   n3 = Apply Additive(
    Base=Apply Additive(
     Base=Blendspace Player 'BS_Tail_Dive'!(X=@n1.SkydiveAimYaw, Y=@n1.SkydiveAimPitch, PlayRate:float=@n2) -> (Pose) {blendSpace=/Game/Accessories/FORT_Tails/Tail_Neon_Cat/Anims/BS_Tail_Dive}
     Additive=Blendspace Player 'BS_Tail_Dive_Zero_Offset'(X=@n1.SkydiveAimYaw, Y=@n1.SkydiveAimPitch) {blendSpace=/Game/Accessories/FORT_Tails/Tail_Neon_Cat/Anims/BS_Tail_Dive_Zero_Offset}
     Alpha=0.5
    ) {alphaBoolBlend {blendOption=Linear}}
    Additive=Blendspace Player 'BS_Tail_Dive_Turn_Additive'(X=Get TailSkydiveYaw(), Y=@n1.SkydiveAimPitch) {blendSpace=/Game/Accessories/FORT_Tails/Tail_Neon_Cat/Anims/BS_Tail_Dive_Turn_Additive}
    Alpha=0.5
   ) {alphaBoolBlend {blendOption=Linear}}
   n4 = Blendspace Player 'BS_Tail_Dive_Lean_Additive'(X=Get TailSkydiveLean(), Y=@n1.SkydiveAimPitch) {blendSpace=/Game/Accessories/FORT_Tails/Tail_Neon_Cat/Anims/BS_Tail_Dive_Lean_Additive}
   n5 = "Transform (Modify) Bone - Bone: C_Root_Main_Root_Jnt"(
    ComponentPose=Local To Component(LocalPose=Apply Additive!(Base=@n3, Additive=@n4, bAlphaBoolEnabled:bool=@n1.bIsSkydiveLeaning) -> (Pose) {alphaInputType=Bool, alphaBoolBlend {blendInTime=0.5, blendOption=Linear, blendOutTime=0.5}})
    Alpha=1
    Translation="0,0,0"
    Rotation="0,0,20"
    Scale="1,1,1"
   ) {
    rotationMode=BMM_Additive
    rotationSpace=BCS_BoneSpace
    alphaBoolBlend {blendOption=Linear}
    boneToModify {boneName=C_Root_Main_Root_Jnt}
    rotation {roll=20}
   }
   Output Animation Pose(Result=Component To Local(ComponentPose=@n5))
  ]
 }
 to_Glide {
  path=AnimGraph.AnimGraphNode_StateMachine_1.FullBody.AnimStateNode_34.SkyDive.AnimGraphNode_StateMachine_0.SkydiveParachute.AnimStateNode_2.SkyDive.AnimGraphNode_StateMachine_0.Skydiving.AnimStateNode_3.to_Glide
  nodes [
   n1 = Break Fort Anim Input Skydiving(FortAnimInput_Skydiving=Get SkydivingState()) -> (
    ParaGilderRootOffset=~
    ParaGliderRootRotationOverride=~
    bUseParaGlideRootModifier=~
    bIsSkydivingFromLaunchPad=~
    bIsSkydivingFromBus=~
    bIsInVortex=~
    bIsDiving=~
    bIsParachuteOpen=~
    bIsSkydiving=~
    bIsParachuteLeaning=~
    bIsSkydiveLeaning=~
    bIsLeaning=~
    bIsSkydiveDiveMode=~
    bPlaySkydiveDrift=~
    bSkydiveDriftDelayActive=~
    bSkydiveDriftAnimAllowed=~
    bIsGliderRight=~
    bIsGliderCenter=~
    bIsGliderLeft=~
    bIsGliderForward=~
    bIsGliderBack=~
    bHasDeployedAndIsStandingUpright=~
    LocalAccelForward=~
    LocalAccelRight=~
    LocalVelocityRight=~
    SkydiveAimPitch=~
    SkydiveAimYaw=~
    DeployChuteAnimRate=~
    SkydiveDriftAnimRate=~
    SkydiveDriftAnimRateCurrent=~
    SkydiveFidgetAnimRate=~
    SkydiveFidgetAnimRateCurrent=~
    SkydiveAdditiveAlpha=~
    SkydiveDriftDelay=~
   )
   n2 = Apply Additive!(
    Base=Blendspace Player 'BS_Tail_Dive_to_Glide'(X=@n1.SkydiveAimYaw, Y=@n1.SkydiveAimPitch) {bLoop=false, blendSpace=/Game/Accessories/FORT_Tails/Tail_Neon_Cat/Anims/BS_Tail_Dive_to_Glide}
    Additive="Sequence Player (additive) 'Tail_Glide_Zero_Offset_Pose'"() {sequence=&Tail_Glide_Zero_Offset_Pose}
    AlphaCurveName:Name=Transition
   ) -> (Pose) {alphaInputType=Curve, alphaBoolBlend {blendOption=Linear}}
   n3 = Blendspace Player 'BS_Tail_Dive_Turn_Additive'(X=@n1.SkydiveAimYaw, Y=@n1.SkydiveAimPitch) {blendSpace=&BS_Tail_Dive_Turn_Additive}
   n4 = Get TailSkydiveLean()
   Output Animation Pose(Result=Apply Additive!(
     Base=Apply Additive!(Base=@n2, Additive=@n3, Alpha:float) -> (Pose) {alphaBoolBlend {blendOption=Linear}}
     Additive=Blendspace Player 'BS_Tail_Glide_Lean_Additive'(X=@n4, Y=@n1.SkydiveAimPitch) {blendSpace=&BS_Tail_Glide_Lean_Additive}
     bAlphaBoolEnabled:bool=@n1.bIsSkydiveLeaning
    ) -> (Pose) {alphaInputType=Bool, alphaBoolBlend {blendInTime=0.5, blendOption=Linear, blendOutTime=0.5}})
  ]
 }
 to_Dive {
  path=AnimGraph.AnimGraphNode_StateMachine_1.FullBody.AnimStateNode_34.SkyDive.AnimGraphNode_StateMachine_0.SkydiveParachute.AnimStateNode_2.SkyDive.AnimGraphNode_StateMachine_0.Skydiving.AnimStateNode_4.to_Dive
  nodes [
   n1 = Break Fort Anim Input Skydiving(FortAnimInput_Skydiving=Get SkydivingState()) -> (
    ParaGilderRootOffset=~
    ParaGliderRootRotationOverride=~
    bUseParaGlideRootModifier=~
    bIsSkydivingFromLaunchPad=~
    bIsSkydivingFromBus=~
    bIsInVortex=~
    bIsDiving=~
    bIsParachuteOpen=~
    bIsSkydiving=~
    bIsParachuteLeaning=~
    bIsSkydiveLeaning=~
    bIsLeaning=~
    bIsSkydiveDiveMode=~
    bPlaySkydiveDrift=~
    bSkydiveDriftDelayActive=~
    bSkydiveDriftAnimAllowed=~
    bIsGliderRight=~
    bIsGliderCenter=~
    bIsGliderLeft=~
    bIsGliderForward=~
    bIsGliderBack=~
    bHasDeployedAndIsStandingUpright=~
    LocalAccelForward=~
    LocalAccelRight=~
    LocalVelocityRight=~
    SkydiveAimPitch=~
    SkydiveAimYaw=~
    DeployChuteAnimRate=~
    SkydiveDriftAnimRate=~
    SkydiveDriftAnimRateCurrent=~
    SkydiveFidgetAnimRate=~
    SkydiveFidgetAnimRateCurrent=~
    SkydiveAdditiveAlpha=~
    SkydiveDriftDelay=~
   )
   n2 = Blendspace Player 'BS_Tail_Glide_to_Dive'(X=@n1.SkydiveAimYaw, Y=@n1.SkydiveAimPitch) {bLoop=false, blendSpace=/Game/Accessories/FORT_Tails/Tail_Neon_Cat/Anims/BS_Tail_Glide_to_Dive}
   n3 = Blendspace Player 'BS_Tail_Glide_Turn_Additive'(X=@n1.SkydiveAimYaw, Y=@n1.SkydiveAimPitch) {blendSpace=&BS_Tail_Glide_Turn_Additive}
   n4 = Get TailSkydiveLean()
   Output Animation Pose(Result=Apply Additive!(
     Base=Apply Additive!(Base=@n2, Additive=@n3, Alpha:float) -> (Pose) {alphaBoolBlend {blendOption=Linear}}
     Additive=Blendspace Player 'BS_Tail_Glide_Lean_Additive'(X=@n4, Y=@n1.SkydiveAimPitch) {blendSpace=&BS_Tail_Glide_Lean_Additive}
     bAlphaBoolEnabled:bool=@n1.bIsSkydiveLeaning
    ) -> (Pose) {alphaInputType=Bool, alphaBoolBlend {blendInTime=0.5, blendOption=Linear, blendOutTime=0.5}})
  ]
 }
 Transition {
  path=AnimGraph.AnimGraphNode_StateMachine_1.FullBody.AnimStateNode_34.SkyDive.AnimGraphNode_StateMachine_0.SkydiveParachute.AnimStateNode_2.SkyDive.AnimGraphNode_StateMachine_0.Skydiving.AnimStateTransitionNode_0.Transition
  nodes [Result(bCanEnterTransition=Break Fort Anim Input Skydiving(FortAnimInput_Skydiving=Get SkydivingState()) -> (
     ParaGilderRootOffset=~
     ParaGliderRootRotationOverride=~
     bUseParaGlideRootModifier=~
     bIsSkydivingFromLaunchPad=~
     bIsSkydivingFromBus=~
     bIsInVortex=~
     bIsDiving=~
     bIsParachuteOpen=~
     bIsSkydiving=~
     bIsParachuteLeaning=~
     bIsSkydiveLeaning=~
     bIsLeaning=~
     bIsSkydiveDiveMode=~
     bPlaySkydiveDrift=~
     bSkydiveDriftDelayActive=~
     bSkydiveDriftAnimAllowed=~
     bIsGliderRight=~
     bIsGliderCenter=~
     bIsGliderLeft=~
     bIsGliderForward=~
     bIsGliderBack=~
     bHasDeployedAndIsStandingUpright=~
     LocalAccelForward=~
     LocalAccelRight=~
     LocalVelocityRight=~
     SkydiveAimPitch=~
     SkydiveAimYaw=~
     DeployChuteAnimRate=~
     SkydiveDriftAnimRate=~
     SkydiveDriftAnimRateCurrent=~
     SkydiveFidgetAnimRate=~
     SkydiveFidgetAnimRateCurrent=~
     SkydiveAdditiveAlpha=~
     SkydiveDriftDelay=~
    ).bIsSkydiveDiveMode)]
 }
 Transition {
  path=AnimGraph.AnimGraphNode_StateMachine_1.FullBody.AnimStateNode_34.SkyDive.AnimGraphNode_StateMachine_0.SkydiveParachute.AnimStateNode_2.SkyDive.AnimGraphNode_StateMachine_0.Skydiving.AnimStateTransitionNode_1.Transition
  nodes [
   n1 = Get SkydivingState()
   Result(bCanEnterTransition=NOT Boolean(A=Break Fort Anim Input Skydiving(FortAnimInput_Skydiving=@n1) -> (
      ParaGilderRootOffset=~
      ParaGliderRootRotationOverride=~
      bUseParaGlideRootModifier=~
      bIsSkydivingFromLaunchPad=~
      bIsSkydivingFromBus=~
      bIsInVortex=~
      bIsDiving=~
      bIsParachuteOpen=~
      bIsSkydiving=~
      bIsParachuteLeaning=~
      bIsSkydiveLeaning=~
      bIsLeaning=~
      bIsSkydiveDiveMode=~
      bPlaySkydiveDrift=~
      bSkydiveDriftDelayActive=~
      bSkydiveDriftAnimAllowed=~
      bIsGliderRight=~
      bIsGliderCenter=~
      bIsGliderLeft=~
      bIsGliderForward=~
      bIsGliderBack=~
      bHasDeployedAndIsStandingUpright=~
      LocalAccelForward=~
      LocalAccelRight=~
      LocalVelocityRight=~
      SkydiveAimPitch=~
      SkydiveAimYaw=~
      DeployChuteAnimRate=~
      SkydiveDriftAnimRate=~
      SkydiveDriftAnimRateCurrent=~
      SkydiveFidgetAnimRate=~
      SkydiveFidgetAnimRateCurrent=~
      SkydiveAdditiveAlpha=~
      SkydiveDriftDelay=~
     ).bIsSkydiveDiveMode))
  ]
 }
 Transition {
  path=AnimGraph.AnimGraphNode_StateMachine_1.FullBody.AnimStateNode_34.SkyDive.AnimGraphNode_StateMachine_0.SkydiveParachute.AnimStateNode_2.SkyDive.AnimGraphNode_StateMachine_0.Skydiving.AnimStateTransitionNode_2.Transition
  nodes [Result()]
 }
 Transition {
  path=AnimGraph.AnimGraphNode_StateMachine_1.FullBody.AnimStateNode_34.SkyDive.AnimGraphNode_StateMachine_0.SkydiveParachute.AnimStateNode_2.SkyDive.AnimGraphNode_StateMachine_0.Skydiving.AnimStateTransitionNode_3.Transition
  nodes [
   n1 = Get SkydivingState()
   Result(bCanEnterTransition=NOT Boolean(A=Break Fort Anim Input Skydiving(FortAnimInput_Skydiving=@n1) -> (
      ParaGilderRootOffset=~
      ParaGliderRootRotationOverride=~
      bUseParaGlideRootModifier=~
      bIsSkydivingFromLaunchPad=~
      bIsSkydivingFromBus=~
      bIsInVortex=~
      bIsDiving=~
      bIsParachuteOpen=~
      bIsSkydiving=~
      bIsParachuteLeaning=~
      bIsSkydiveLeaning=~
      bIsLeaning=~
      bIsSkydiveDiveMode=~
      bPlaySkydiveDrift=~
      bSkydiveDriftDelayActive=~
      bSkydiveDriftAnimAllowed=~
      bIsGliderRight=~
      bIsGliderCenter=~
      bIsGliderLeft=~
      bIsGliderForward=~
      bIsGliderBack=~
      bHasDeployedAndIsStandingUpright=~
      LocalAccelForward=~
      LocalAccelRight=~
      LocalVelocityRight=~
      SkydiveAimPitch=~
      SkydiveAimYaw=~
      DeployChuteAnimRate=~
      SkydiveDriftAnimRate=~
      SkydiveDriftAnimRateCurrent=~
      SkydiveFidgetAnimRate=~
      SkydiveFidgetAnimRateCurrent=~
      SkydiveAdditiveAlpha=~
      SkydiveDriftDelay=~
     ).bIsSkydiveDiveMode))
  ]
 }
 Transition {
  path=AnimGraph.AnimGraphNode_StateMachine_1.FullBody.AnimStateNode_34.SkyDive.AnimGraphNode_StateMachine_0.SkydiveParachute.AnimStateNode_2.SkyDive.AnimGraphNode_StateMachine_0.Skydiving.AnimStateTransitionNode_4.Transition
  nodes [Result(bCanEnterTransition=Break Fort Anim Input Skydiving(FortAnimInput_Skydiving=Get SkydivingState()) -> (
     ParaGilderRootOffset=~
     ParaGliderRootRotationOverride=~
     bUseParaGlideRootModifier=~
     bIsSkydivingFromLaunchPad=~
     bIsSkydivingFromBus=~
     bIsInVortex=~
     bIsDiving=~
     bIsParachuteOpen=~
     bIsSkydiving=~
     bIsParachuteLeaning=~
     bIsSkydiveLeaning=~
     bIsLeaning=~
     bIsSkydiveDiveMode=~
     bPlaySkydiveDrift=~
     bSkydiveDriftDelayActive=~
     bSkydiveDriftAnimAllowed=~
     bIsGliderRight=~
     bIsGliderCenter=~
     bIsGliderLeft=~
     bIsGliderForward=~
     bIsGliderBack=~
     bHasDeployedAndIsStandingUpright=~
     LocalAccelForward=~
     LocalAccelRight=~
     LocalVelocityRight=~
     SkydiveAimPitch=~
     SkydiveAimYaw=~
     DeployChuteAnimRate=~
     SkydiveDriftAnimRate=~
     SkydiveDriftAnimRateCurrent=~
     SkydiveFidgetAnimRate=~
     SkydiveFidgetAnimRateCurrent=~
     SkydiveAdditiveAlpha=~
     SkydiveDriftDelay=~
    ).bIsSkydiveDiveMode)]
 }
 Transition {
  path=AnimGraph.AnimGraphNode_StateMachine_1.FullBody.AnimStateNode_34.SkyDive.AnimGraphNode_StateMachine_0.SkydiveParachute.AnimStateNode_2.SkyDive.AnimGraphNode_StateMachine_0.Skydiving.AnimStateTransitionNode_6.Transition
  nodes [
   n1 = Get SkydivingState()
   Result(bCanEnterTransition=NOT Boolean(A=Break Fort Anim Input Skydiving(FortAnimInput_Skydiving=@n1) -> (
      ParaGilderRootOffset=~
      ParaGliderRootRotationOverride=~
      bUseParaGlideRootModifier=~
      bIsSkydivingFromLaunchPad=~
      bIsSkydivingFromBus=~
      bIsInVortex=~
      bIsDiving=~
      bIsParachuteOpen=~
      bIsSkydiving=~
      bIsParachuteLeaning=~
      bIsSkydiveLeaning=~
      bIsLeaning=~
      bIsSkydiveDiveMode=~
      bPlaySkydiveDrift=~
      bSkydiveDriftDelayActive=~
      bSkydiveDriftAnimAllowed=~
      bIsGliderRight=~
      bIsGliderCenter=~
      bIsGliderLeft=~
      bIsGliderForward=~
      bIsGliderBack=~
      bHasDeployedAndIsStandingUpright=~
      LocalAccelForward=~
      LocalAccelRight=~
      LocalVelocityRight=~
      SkydiveAimPitch=~
      SkydiveAimYaw=~
      DeployChuteAnimRate=~
      SkydiveDriftAnimRate=~
      SkydiveDriftAnimRateCurrent=~
      SkydiveFidgetAnimRate=~
      SkydiveFidgetAnimRateCurrent=~
      SkydiveAdditiveAlpha=~
      SkydiveDriftDelay=~
     ).bIsSkydiveDiveMode))
  ]
 }
 Transition {
  path=AnimGraph.AnimGraphNode_StateMachine_1.FullBody.AnimStateNode_34.SkyDive.AnimGraphNode_StateMachine_0.SkydiveParachute.AnimStateNode_2.SkyDive.AnimGraphNode_StateMachine_0.Skydiving.AnimStateTransitionNode_7.Transition
  nodes [Result(bCanEnterTransition=Break Fort Anim Input Skydiving(FortAnimInput_Skydiving=Get SkydivingState()) -> (
     ParaGilderRootOffset=~
     ParaGliderRootRotationOverride=~
     bUseParaGlideRootModifier=~
     bIsSkydivingFromLaunchPad=~
     bIsSkydivingFromBus=~
     bIsInVortex=~
     bIsDiving=~
     bIsParachuteOpen=~
     bIsSkydiving=~
     bIsParachuteLeaning=~
     bIsSkydiveLeaning=~
     bIsLeaning=~
     bIsSkydiveDiveMode=~
     bPlaySkydiveDrift=~
     bSkydiveDriftDelayActive=~
     bSkydiveDriftAnimAllowed=~
     bIsGliderRight=~
     bIsGliderCenter=~
     bIsGliderLeft=~
     bIsGliderForward=~
     bIsGliderBack=~
     bHasDeployedAndIsStandingUpright=~
     LocalAccelForward=~
     LocalAccelRight=~
     LocalVelocityRight=~
     SkydiveAimPitch=~
     SkydiveAimYaw=~
     DeployChuteAnimRate=~
     SkydiveDriftAnimRate=~
     SkydiveDriftAnimRateCurrent=~
     SkydiveFidgetAnimRate=~
     SkydiveFidgetAnimRateCurrent=~
     SkydiveAdditiveAlpha=~
     SkydiveDriftDelay=~
    ).bIsSkydiveDiveMode)]
 }
 Transition {
  path=AnimGraph.AnimGraphNode_StateMachine_1.FullBody.AnimStateNode_34.SkyDive.AnimGraphNode_StateMachine_0.SkydiveParachute.AnimStateNode_2.SkyDive.AnimGraphNode_StateMachine_0.Skydiving.AnimStateTransitionNode_8.Transition
  nodes [Result()]
 }
 SkydiveStart {
  path=AnimGraph.AnimGraphNode_StateMachine_1.FullBody.AnimStateNode_34.SkyDive.AnimGraphNode_StateMachine_0.SkydiveParachute.AnimStateNode_28.SkydiveStart
  nodes [Output Animation Pose(Result=Sequence Player 'Tail_SkydiveStart'() {bLoopAnimation=false, sequence=/Game/Accessories/FORT_Tails/Common/Poses/Tail_SkydiveStart})]
 }
 Transition {
  path=AnimGraph.AnimGraphNode_StateMachine_1.FullBody.AnimStateNode_34.SkyDive.AnimGraphNode_StateMachine_0.SkydiveParachute.AnimStateTransitionNode_0.Transition
  nodes [Result(bCanEnterTransition=Get bIsParachuteOpened())]
 }
 Transition {
  path=AnimGraph.AnimGraphNode_StateMachine_1.FullBody.AnimStateNode_34.SkyDive.AnimGraphNode_StateMachine_0.SkydiveParachute.AnimStateTransitionNode_1.Transition
  nodes [Result(bCanEnterTransition=NOT Boolean(A=Get bIsParachuteOpened()))]
 }
 Transition {
  path=AnimGraph.AnimGraphNode_StateMachine_1.FullBody.AnimStateNode_34.SkyDive.AnimGraphNode_StateMachine_0.SkydiveParachute.AnimStateTransitionNode_28.Transition
  nodes [Result()]
 }
 DBNO {
  path=AnimGraph.AnimGraphNode_StateMachine_1.FullBody.AnimStateNode_9.DBNO
  nodes [Output Animation Pose(Result=Sequence Player 'Tail_DBNO'() {sequence=/Game/Accessories/FORT_Tails/Tail_Neon_Cat/Anims/Tail_DBNO})]
 }
 Transition {
  path=AnimGraph.AnimGraphNode_StateMachine_1.FullBody.AnimStateTransitionNode_0.Transition
  nodes [Result(bCanEnterTransition=Get bIsSkyDiving())]
 }
 Transition {
  path=AnimGraph.AnimGraphNode_StateMachine_1.FullBody.AnimStateTransitionNode_1.Transition
  nodes [Result(bCanEnterTransition=NOT Boolean(A=Get bIsSkyDiving()))]
 }
 Transition {
  path=AnimGraph.AnimGraphNode_StateMachine_1.FullBody.AnimStateTransitionNode_32.Transition
  nodes [
   Result(bCanEnterTransition=Get bIsCrouching())
   OR Boolean(A=Get bTransition_Idle_Crouching() -> (bTransition_Idle_Crouching:bool), B=false)
  ]
 }
 AnimationTransitionGraph_1 {
  path=AnimGraph.AnimGraphNode_StateMachine_1.FullBody.AnimStateTransitionNode_74.AnimationTransitionGraph_1
  nodes [Result(bCanEnterTransition=OR Boolean(A=Get bTransition_Idle_Skydive() -> (bTransition_Idle_Skydive:bool), B=Get bIsParachuteOpened()))]
 }
 Transition {
  path=AnimGraph.AnimGraphNode_StateMachine_1.FullBody.AnimStateTransitionNode_79.Transition
  nodes [Result(bCanEnterTransition=Get bTransition_Skydive_Idle() -> (bTransition_Skydive_Idle:bool))]
 }
 Transition {
  path=AnimGraph.AnimGraphNode_StateMachine_1.FullBody.AnimStateTransitionNode_8.Transition
  nodes [Result(bCanEnterTransition=Get bTransition_Locomotion_Idle() -> (bTransition_Locomotion_Idle:bool))]
 }
 Transition {
  path=AnimGraph.AnimGraphNode_StateMachine_1.FullBody.AnimStateTransitionNode_80.Transition
  nodes [Result(bCanEnterTransition=Get bTransition_Idle_Locomotion() -> (bTransition_Idle_Locomotion:bool))]
 }
 Transition {
  path=AnimGraph.AnimGraphNode_StateMachine_1.FullBody.AnimStateTransitionNode_81.Transition
  nodes [Result(bCanEnterTransition=Get bIsDBNO())]
 }
 Transition {
  path=AnimGraph.AnimGraphNode_StateMachine_1.FullBody.AnimStateTransitionNode_90.Transition
  nodes [Result(bCanEnterTransition=NOT Boolean(A=Get bIsDBNO()))]
 }
 Transition {
  path=AnimGraph.AnimGraphNode_StateMachine_1.FullBody.AnimStateTransitionNode_92.Transition
  nodes [Result(bCanEnterTransition=Get bIsCrouching())]
 }
 Transition {
  path=AnimGraph.AnimGraphNode_StateMachine_1.FullBody.AnimStateTransitionNode_93.Transition
  nodes [Result(bCanEnterTransition=Get bTransition_Crouching_Locomotion() -> (bTransition_Crouching_Locomotion:bool))]
 }
 Transition {
  path=AnimGraph.AnimGraphNode_StateMachine_1.FullBody.AnimStateTransitionNode_95.Transition
  nodes [Result(bCanEnterTransition=NOT Boolean(A=Get bIsCrouching()))]
 }
 EventGraph {nodes [
   Event OnCharacterCustomizationLoaded()
   n1 = CheckForBackbling()
   > n2 = Branch(Condition=OR Boolean(A=@n1.bIsBackblingACape, B=@n1.bIsBackblingLong))
   > n3 = Get Dynamic Material Instance(SkeletalMeshComponent=GetPartSkeletalMeshComponent())
   > SetScalarParameterValue(self=@n3, ParameterName=Hide Element 01, Value=1)
   SetScalarParameterValue(execute=@n2.else, self=@n3, ParameterName=Hide Element 01, Value=0)
   Event BlueprintUpdateAnimation()
   > n4 = Sequence!(execute) -> (then_0, then_1, then_2:Exec, then_3:Exec, then_4:Exec, then_5:Exec)
   > n5 = Branch(Condition=Get bIsSkyDiving())
   > Set SkydiveAlpha(SkydiveAlpha:double=1) -> (Output_Get:double)
   > Set AnimDynamicsAlpha(AnimDynamicsAlpha:double=0) -> (Output_Get:double)
   > n6 = Branch(Condition=GliderItemDefinitionHasAnyMetaTag!(Tags:Array of Gameplay Tag Structures=Make Array("[0]"=(TagName="Cosmetics.MetaTags.CoconutShell")), bRequiresOpened:bool=true) -> (ReturnValue:bool))
   > Set matchingGliderDeployed(matchingGliderDeployed:bool=true) -> (Output_Get:bool)
   Set matchingGliderDeployed(execute=@n6.else, matchingGliderDeployed=false)
   Set SkydiveAlpha(execute=@n5.else, SkydiveAlpha=0)
   > n7 = Branch(Condition=Get IsOnZipline() -> (IsOnZipline:bool))
   > Set AnimDynamicsAlpha(AnimDynamicsAlpha=0)
   Set AnimDynamicsAlpha(execute=@n7.else, AnimDynamicsAlpha=1)
   > Set matchingGliderDeployed(matchingGliderDeployed=false)
   n8 = Branch(execute=@n4.then_1, Condition=Get bIsSkydiveDiveMode())
   > Set DivingAlpha(DivingAlpha:double=1) -> (Output_Get:double)
   Set DivingAlpha(execute=@n8.else, DivingAlpha=0)
   n9 = Branch(execute=@n4.then_2, Condition=Get bIsParachuteOpened())
   > Set ParachutingAlpha(ParachutingAlpha:double=1) -> (Output_Get:double)
   Set ParachutingAlpha(execute=@n9.else, ParachutingAlpha=0)
   n10 = Branch(execute=@n4.then_3, Condition=Get bIsSkydiveFloating())
   > Set SkydiveFloatingAlpha(SkydiveFloatingAlpha:double=1) -> (Output_Get:double)
   Set SkydiveFloatingAlpha(execute=@n10.else, SkydiveFloatingAlpha=0)
   n11 = Branch(execute=@n4.then_4, Condition=Get bIsSliding())
   > Set SlideOffsetAlpha(SlideOffsetAlpha:double=1) -> (Output_Get:double)
   > Set TacSprintAlpha(TacSprintAlpha:double=0) -> (Output_Get:double)
   Set SlideOffsetAlpha(execute=@n11.else, SlideOffsetAlpha=0)
   > n12 = Branch(Condition=OR Boolean(A=Get bIsTacticalSprinting(), B=Get IsOnZipline()))
   > Set TacSprintAlpha(TacSprintAlpha=1)
   Set TacSprintAlpha(execute=@n12.else, TacSprintAlpha=0)
   Branch(execute=@n4.then_5, Condition=Get bIsInFrontEnd())
   > n13 = Branch(Condition=Get bIsPlayingEmote())
   Set Custom Idle(execute=[@n15.then, @n13.then], Custom Idle:bool=false) -> (Output_Get:bool)
   n14 = GetOwnerFrontEndAnimInstance()
   n15 = Branch(execute=@n13.else, Condition="Equal (Float)"(A=GetCurveValue(self=@n14, CurveName=Frontend_Backpack_Pose), B=1))
   Set Custom Idle(execute=@n15.else, Custom Idle="Equal (Float)"(A=GetCurveValue(self=@n14, CurveName=Frontend_Idle_Bypass), B=1))
   n16 = Event BlueprintProcessAnimInputEvent()
   > n17 = Branch(Condition=IsNamed(self=@n16.AnimInputEvent, InEventName=aie.Player.Zipline.Entered))
   > Set IsOnZipline(IsOnZipline:bool=true) -> (Output_Get:bool)
   Branch(execute=@n17.else, Condition=IsNamed(self=@n16.AnimInputEvent, InEventName=aie.Player.Zipline.Exited))
   > Set IsOnZipline(IsOnZipline=false)
   n18 = Sequence!(execute) -> (then_0, then_1, then_2, then_3)
   > n19 = Branch(Condition=Get bIsSkydiveFloating())
   > Set AnimDynamicsAlpha(AnimDynamicsAlpha=1)
   Set AnimDynamicsAlpha(execute=@n19.else, AnimDynamicsAlpha=0)
   n20 = Get bIsParachuteOpened()
   n21 = Get bIsSkyDiving()
   n22 = Branch(execute=@n18.then_1, Condition=OR Boolean(A=OR Boolean(A=@n20, B=@n21), B=Get bIsInFrontEnd()))
   > n23 = Branch(Condition=Get bIsSkydiveDiveMode())
   > Set SkydiveFloatingAlpha(SkydiveFloatingAlpha=0)
   > Set DivingAlpha(DivingAlpha=1)
   Set SkydiveFloatingAlpha(execute=@n23.else, SkydiveFloatingAlpha=0.1)
   > Set DivingAlpha(DivingAlpha=0)
   Set SkydiveFloatingAlpha(execute=@n22.else, SkydiveFloatingAlpha=1)
  ]}
]
variables [
 AnimDynamicsAlpha {type=number}
 Custom Idle {type=boolean}
 DivingAlpha {type=number}
 IsOnZipline {type=boolean}
 matchingGliderDeployed {type=boolean}
 ParachutingAlpha {type=number}
 SkydiveAlpha {type=number}
 SkydiveFloatingAlpha {type=number}
 SlideOffsetAlpha {type=number}
 TacSprintAlpha {type=number}
]
